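# Anim Blueprint for the Quartz backpack cosmetic, driving its spine, neck and rib bones from the wearer's body.
# Body socket world transforms, shifted by a root offset, feed Anim Dynamics and constraints, then a cached pose.
# Gender and large-body offsets, twist correction, skydiving rib rotation and sine breathing that fades with speed adjust that pose.
# An additive ShutterDown animation plays on chest opening, eliminations or SlurpJuice shield gains.
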AnimBlueprint /BRCosmetics/Accessories/FORT_Backpacks/Backpack_M_MED_Quartz/Meshes/M_MED_Quartz_Pack_AnimBP
target_skeleton=&/M_MED_Quartz_Pack_Skeleton
graphs [
 AnimGraph {nodes [
   n1 = "Transform (Modify) Bone - Bone: dyn_spine_03"(
    ComponentPose=Local To Component(LocalPose=Use cached pose 'PostConstraint'())
    Alpha=1
    Translation="0,0,0.7"
    Rotation="0,0,0"
    Scale="1,1,1"
   ) {
    scaleSpace=BCS_BoneSpace
    translationMode=BMM_Additive
    translationSpace=BCS_BoneSpace
    alphaBoolBlend {blendOption=Linear}
    alphaScaleBiasClamp {inRange {max=0, min=-40}, outRange {max=0, min=1}}
    boneToModify {boneName=dyn_spine_03}
    translation {z=0.7}
   }
   n2 = "Transform (Modify) Bone - Bone: dyn_spine_04"(ComponentPose=@n1, Alpha=1, Translation="0,0,2", Rotation="0,0,0", Scale="1,1,1") {
    scaleSpace=BCS_BoneSpace
    translationMode=BMM_Additive
    translationSpace=BCS_BoneSpace
    alphaBoolBlend {blendOption=Linear}
    alphaScaleBiasClamp {inRange {max=0, min=-40}, outRange {max=0, min=1}}
    boneToModify {boneName=dyn_spine_04}
    translation {z=2}
   }
   n3 = "Transform (Modify) Bone - Bone: dyn_neck_01"(ComponentPose=@n2, Alpha=1, Translation="0,-1,-0.5", Rotation="0,0,0", Scale="1,1,1") {
    scaleSpace=BCS_BoneSpace
    translationMode=BMM_Additive
    translationSpace=BCS_BoneSpace
    alphaBoolBlend {blendOption=Linear}
    alphaScaleBiasClamp {inRange {max=0, min=-40}, outRange {max=0, min=1}}
    boneToModify {boneName=dyn_neck_01}
    translation {y=-1, z=-0.5}
   }
   n4 = "Transform (Modify) Bone - Bone: dyn_neck_02"(ComponentPose=@n3, Alpha=1, Translation="0,0,0", Rotation="0,0,0", Scale="1,1,1") {
    scaleSpace=BCS_BoneSpace
    translationMode=BMM_Additive
    translationSpace=BCS_BoneSpace
    alphaBoolBlend {blendOption=Linear}
    alphaScaleBiasClamp {inRange {max=0, min=-40}, outRange {max=0, min=1}}
    boneToModify {boneName=dyn_neck_02}
   }
   n5 = "Transform (Modify) Bone - Bone: dyn_spine_05"(ComponentPose=@n4, Alpha=1, Translation="0,-6,6.5", Rotation="0,0,3", Scale="1,1,1") {
    rotationMode=BMM_Additive
    rotationSpace=BCS_BoneSpace
    scaleSpace=BCS_BoneSpace
    translationMode=BMM_Additive
    translationSpace=BCS_BoneSpace
    alphaBoolBlend {blendOption=Linear}
    alphaScaleBiasClamp {inRange {max=0, min=-40}, outRange {max=0, min=1}}
    boneToModify {boneName=dyn_spine_05}
    rotation {roll=3}
    translation {y=-6, z=6.5}
   }
   n6 = Component To Local(ComponentPose=@n5)
   n7 = "Transform (Modify) Bone - Bone: dyn_spine_03"(
    ComponentPose=Local To Component(LocalPose=Use cached pose 'PostConstraint'())
    Alpha=1
    Translation="0,0,1.3"
    Rotation="0,0,0"
    Scale="1,1,1"
   ) {
    scaleSpace=BCS_BoneSpace
    translationMode=BMM_Additive
    translationSpace=BCS_BoneSpace
    alphaBoolBlend {blendOption=Linear}
    alphaScaleBiasClamp {inRange {max=0, min=-40}, outRange {max=0, min=1}}
    boneToModify {boneName=dyn_spine_03}
    translation {z=1.3}
   }
   n8 = "Transform (Modify) Bone - Bone: dyn_spine_04"(ComponentPose=@n7, Alpha=1, Translation="0,0,1.2", Rotation="0,0,0", Scale="1,1,1") {
    scaleSpace=BCS_BoneSpace
    translationMode=BMM_Additive
    translationSpace=BCS_BoneSpace
    alphaBoolBlend {blendOption=Linear}
    alphaScaleBiasClamp {inRange {max=0, min=-40}, outRange {max=0, min=1}}
    boneToModify {boneName=dyn_spine_04}
    translation {z=1.2}
   }
   n9 = "Transform (Modify) Bone - Bone: dyn_neck_01"(ComponentPose=@n8, Alpha=1, Translation="0,0,-0.1", Rotation="0,0,0", Scale="1,1,1") {
    scaleSpace=BCS_BoneSpace
    translationMode=BMM_Additive
    translationSpace=BCS_BoneSpace
    alphaBoolBlend {blendOption=Linear}
    alphaScaleBiasClamp {inRange {max=0, min=-40}, outRange {max=0, min=1}}
    boneToModify {boneName=dyn_neck_01}
    translation {z=-0.1}
   }
   n10 = "Transform (Modify) Bone - Bone: dyn_neck_02"(ComponentPose=@n9, Alpha=1, Translation="0,0,0", Rotation="0,0,0", Scale="1,1,1") {
    scaleSpace=BCS_BoneSpace
    translationMode=BMM_Additive
    translationSpace=BCS_BoneSpace
    alphaBoolBlend {blendOption=Linear}
    alphaScaleBiasClamp {inRange {max=0, min=-40}, outRange {max=0, min=1}}
    boneToModify {boneName=dyn_neck_02}
   }
   n11 = "Transform (Modify) Bone - Bone: dyn_spine_05"(ComponentPose=@n10, Alpha=1, Translation="0,-3,7.4", Rotation="0,0,3", Scale="1,1,1") {
    rotationMode=BMM_Additive
    rotationSpace=BCS_ParentBoneSpace
    scaleSpace=BCS_BoneSpace
    translationMode=BMM_Additive
    translationSpace=BCS_BoneSpace
    alphaBoolBlend {blendOption=Linear}
    alphaScaleBiasClamp {inRange {max=0, min=-40}, outRange {max=0, min=1}}
    boneToModify {boneName=dyn_spine_05}
    rotation {roll=3}
    translation {y=-3, z=7.4}
   }
   n12 = Component To Local(ComponentPose=@n11)
   n13 = Get isFemale() -> (isFemale:bool)
   n14 = "Transform (Modify) Bone - Bone: dyn_spine_03"(
    ComponentPose=Local To Component(LocalPose=Blend Poses by bool(BlendPose_0=@n12, BlendPose_1=@n6, bActiveValue=@n13) {blendType=Linear})
    Alpha=Get dyn_Spine_02_alpha_correction() -> (dyn_Spine_02_alpha_correction:double)
    Translation="0,-2,1"
    Rotation="0,0,14"
    Scale="1,1,1"
   ) {
    rotationMode=BMM_Additive
    rotationSpace=BCS_BoneSpace
    scaleSpace=BCS_BoneSpace
    translationMode=BMM_Additive
    translationSpace=BCS_BoneSpace
    alphaBoolBlend {blendOption=Linear}
    alphaScaleBiasClamp {inRange {max=0, min=-40}, outRange {max=0, min=1}}
    boneToModify {boneName=dyn_spine_03}
    rotation {roll=14}
    translation {y=-2, z=1}
   }
   n15 = "Transform (Modify) Bone - Bone: dyn_spine_02"!(ComponentPose=@n14, bAlphaBoolEnabled:bool=false, Translation="0,0,0", Rotation="0,0,14", Scale="1,1,1") -> (Pose) {
    alphaInputType=Bool
    rotationMode=BMM_Additive
    rotationSpace=BCS_BoneSpace
    scaleSpace=BCS_BoneSpace
    translationSpace=BCS_BoneSpace
    alphaBoolBlend {blendOption=Linear}
    alphaScaleBiasClamp {inRange {max=0, min=-40}, outRange {max=0, min=1}}
    boneToModify {boneName=dyn_spine_02}
    rotation {roll=14}
   }
   n16 = Get SkyDivingWihtoutParachute() -> (SkyDivingWihtoutParachute:bool)
   n17 = "Transform (Modify) Bone - Bone: dyn_spine_05_rib_r"!(ComponentPose=@n15, bAlphaBoolEnabled:bool=@n16, Translation="0,0,0", Rotation="0,20,0", Scale="1,1,1") -> (Pose) {
    alphaInputType=Bool
    rotationMode=BMM_Additive
    rotationSpace=BCS_BoneSpace
    scaleSpace=BCS_BoneSpace
    translationSpace=BCS_BoneSpace
    alphaBoolBlend {blendInTime=0.1, blendOption=Linear, blendOutTime=0.1}
    alphaScaleBiasClamp {inRange {max=0, min=-40}, outRange {max=0, min=1}}
    boneToModify {boneName=dyn_spine_05_rib_r}
    rotation {yaw=20}
   }
   n18 = "Transform (Modify) Bone - Bone: dyn_spine_04_rib_r"!(ComponentPose=@n17, bAlphaBoolEnabled:bool=@n16, Translation="0,0,0", Rotation="0,20,0", Scale="1,1,1") -> (Pose) {
    alphaInputType=Bool
    rotationMode=BMM_Additive
    rotationSpace=BCS_BoneSpace
    scaleSpace=BCS_BoneSpace
    translationSpace=BCS_BoneSpace
    alphaBoolBlend {blendInTime=0.1, blendOption=Linear, blendOutTime=0.1}
    alphaScaleBiasClamp {inRange {max=0, min=-40}, outRange {max=0, min=1}}
    boneToModify {boneName=dyn_spine_04_rib_r}
    rotation {yaw=20}
   }
   n19 = "Transform (Modify) Bone - Bone: dyn_spine_03_rib_r"!(ComponentPose=@n18, bAlphaBoolEnabled:bool=@n16, Translation="0,0,0", Rotation="0,10,0", Scale="1,1,1") -> (Pose) {
    alphaInputType=Bool
    rotationMode=BMM_Additive
    rotationSpace=BCS_BoneSpace
    scaleSpace=BCS_BoneSpace
    translationSpace=BCS_BoneSpace
    alphaBoolBlend {blendInTime=0.1, blendOption=Linear, blendOutTime=0.1}
    alphaScaleBiasClamp {inRange {max=0, min=-40}, outRange {max=0, min=1}}
    boneToModify {boneName=dyn_spine_03_rib_r}
    rotation {yaw=10}
   }
   n20 = "Transform (Modify) Bone - Bone: dyn_spine_05_lib_l"!(ComponentPose=@n19, bAlphaBoolEnabled:bool=@n16, Translation="0,0,0", Rotation="0,20,0", Scale="1,1,1") -> (Pose) {
    alphaInputType=Bool
    rotationMode=BMM_Additive
    rotationSpace=BCS_BoneSpace
    scaleSpace=BCS_BoneSpace
    translationSpace=BCS_BoneSpace
    alphaBoolBlend {blendInTime=0.1, blendOption=Linear, blendOutTime=0.1}
    alphaScaleBiasClamp {inRange {max=0, min=-40}, outRange {max=0, min=1}}
    boneToModify {boneName=dyn_spine_05_lib_l}
    rotation {yaw=20}
   }
   n21 = "Transform (Modify) Bone - Bone: dyn_spine_04_lib_l"!(ComponentPose=@n20, bAlphaBoolEnabled:bool=@n16, Translation="0,0,0", Rotation="0,20,0", Scale="1,1,1") -> (Pose) {
    alphaInputType=Bool
    rotationMode=BMM_Additive
    rotationSpace=BCS_BoneSpace
    scaleSpace=BCS_BoneSpace
    translationSpace=BCS_BoneSpace
    alphaBoolBlend {blendInTime=0.1, blendOption=Linear, blendOutTime=0.1}
    alphaScaleBiasClamp {inRange {max=0, min=-40}, outRange {max=0, min=1}}
    boneToModify {boneName=dyn_spine_04_lib_l}
    rotation {yaw=20}
   }
   n22 = "Transform (Modify) Bone - Bone: dyn_spine_03_lib_l"!(ComponentPose=@n21, bAlphaBoolEnabled:bool=@n16, Translation="0,0,0", Rotation="0,10,0", Scale="1,1,1") -> (Pose) {
    alphaInputType=Bool
    rotationMode=BMM_Additive
    rotationSpace=BCS_BoneSpace
    scaleSpace=BCS_BoneSpace
    translationSpace=BCS_BoneSpace
    alphaBoolBlend {blendInTime=0.1, blendOption=Linear, blendOutTime=0.1}
    alphaScaleBiasClamp {inRange {max=0, min=-40}, outRange {max=0, min=1}}
    boneToModify {boneName=dyn_spine_03_lib_l}
    rotation {yaw=10}
   }
   n23 = Get Breathe_alpha() -> (Breathe_alpha:double)
   n24 = "Transform (Modify) Bone - Bone: dyn_spine_05_rib_r"!(ComponentPose=@n22, Alpha:float=@n23, Translation="0,0,0", Rotation="5,10,0", Scale="1,1,1") -> (Pose) {
    rotationMode=BMM_Additive
    rotationSpace=BCS_BoneSpace
    scaleSpace=BCS_BoneSpace
    translationSpace=BCS_BoneSpace
    alphaBoolBlend {blendInTime=0.1, blendOption=Linear, blendOutTime=0.1}
    alphaScaleBiasClamp {bInterpResult=true, inRange {max=0, min=-40}, outRange {max=0, min=1}}
    boneToModify {boneName=dyn_spine_05_rib_r}
    rotation {pitch=5, yaw=10}
   }
   n25 = "Transform (Modify) Bone - Bone: dyn_spine_04_rib_r"!(ComponentPose=@n24, Alpha:float=@n23, Translation="0,0,0", Rotation="5,10,0", Scale="1,1,1") -> (Pose) {
    rotationMode=BMM_Additive
    rotationSpace=BCS_BoneSpace
    scaleSpace=BCS_BoneSpace
    translationSpace=BCS_BoneSpace
    alphaBoolBlend {blendInTime=0.1, blendOption=Linear, blendOutTime=0.1}
    alphaScaleBiasClamp {bInterpResult=true, inRange {max=0, min=-40}, outRange {max=0, min=1}}
    boneToModify {boneName=dyn_spine_04_rib_r}
    rotation {pitch=5, yaw=10}
   }
   n26 = "Transform (Modify) Bone - Bone: dyn_spine_03_rib_r"!(ComponentPose=@n25, Alpha:float=@n23, Translation="0,0,0", Rotation="5,15,0", Scale="1,1,1") -> (Pose) {
    rotationMode=BMM_Additive
    rotationSpace=BCS_BoneSpace
    scaleSpace=BCS_BoneSpace
    translationSpace=BCS_BoneSpace
    alphaBoolBlend {blendInTime=0.1, blendOption=Linear, blendOutTime=0.1}
    alphaScaleBiasClamp {bInterpResult=true, inRange {max=0, min=-40}, outRange {max=0, min=1}}
    boneToModify {boneName=dyn_spine_03_rib_r}
    rotation {pitch=5, yaw=15}
   }
   n27 = "Transform (Modify) Bone - Bone: dyn_spine_02_rib_r"(ComponentPose=@n26, Alpha=@n23, Translation="0,0,0", Rotation="5,15,0", Scale="1,1,1") {
    rotationMode=BMM_Additive
    rotationSpace=BCS_BoneSpace
    scaleSpace=BCS_BoneSpace
    translationSpace=BCS_BoneSpace
    alphaBoolBlend {blendInTime=0.1, blendOption=Linear, blendOutTime=0.1}
    alphaScaleBiasClamp {bInterpResult=true, inRange {max=0, min=-40}, outRange {max=0, min=1}}
    boneToModify {boneName=dyn_spine_02_rib_r}
    rotation {pitch=5, yaw=15}
   }
   n28 = "Transform (Modify) Bone - Bone: dyn_spine_05_lib_l"!(ComponentPose=@n27, Alpha:float=@n23, Translation="0,0,0", Rotation="5,10,0", Scale="1,1,1") -> (Pose) {
    rotationMode=BMM_Additive
    rotationSpace=BCS_BoneSpace
    scaleSpace=BCS_BoneSpace
    translationSpace=BCS_BoneSpace
    alphaBoolBlend {blendInTime=0.1, blendOption=Linear, blendOutTime=0.1}
    alphaScaleBiasClamp {bInterpResult=true, inRange {max=0, min=-40}, outRange {max=0, min=1}}
    boneToModify {boneName=dyn_spine_05_lib_l}
    rotation {pitch=5, yaw=10}
   }
   n29 = "Transform (Modify) Bone - Bone: dyn_spine_04_lib_l"!(ComponentPose=@n28, Alpha:float=@n23, Translation="0,0,0", Rotation="5,10,0", Scale="1,1,1") -> (Pose) {
    rotationMode=BMM_Additive
    rotationSpace=BCS_BoneSpace
    scaleSpace=BCS_BoneSpace
    translationSpace=BCS_BoneSpace
    alphaBoolBlend {blendInTime=0.1, blendOption=Linear, blendOutTime=0.1}
    alphaScaleBiasClamp {bInterpResult=true, inRange {max=0, min=-40}, outRange {max=0, min=1}}
    boneToModify {boneName=dyn_spine_04_lib_l}
    rotation {pitch=5, yaw=10}
   }
   n30 = "Transform (Modify) Bone - Bone: dyn_spine_03_lib_l"!(ComponentPose=@n29, Alpha:float=@n23, Translation="0,0,0", Rotation="5,15,0", Scale="1,1,1") -> (Pose) {
    rotationMode=BMM_Additive
    rotationSpace=BCS_BoneSpace
    scaleSpace=BCS_BoneSpace
    translationSpace=BCS_BoneSpace
    alphaBoolBlend {blendInTime=0.1, blendOption=Linear, blendOutTime=0.1}
    alphaScaleBiasClamp {bInterpResult=true, inRange {max=0, min=-40}, outRange {max=0, min=1}}
    boneToModify {boneName=dyn_spine_03_lib_l}
    rotation {pitch=5, yaw=15}
   }
   n31 = "Transform (Modify) Bone - Bone: dyn_spine_02_lib_l"(ComponentPose=@n30, Alpha=@n23, Translation="0,0,0", Rotation="5,15,0", Scale="1,1,1") {
    rotationMode=BMM_Additive
    rotationSpace=BCS_BoneSpace
    scaleSpace=BCS_BoneSpace
    translationSpace=BCS_BoneSpace
    alphaBoolBlend {blendInTime=0.1, blendOption=Linear, blendOutTime=0.1}
    alphaScaleBiasClamp {bInterpResult=true, inRange {max=0, min=-40}, outRange {max=0, min=1}}
    boneToModify {boneName=dyn_spine_02_lib_l}
    rotation {pitch=5, yaw=15}
   }
   n32 = "Transform (Modify) Bone - Bone: pack"(ComponentPose=@n31, Alpha="To Float (Boolean)"(InBool=Get IsLarge() -> (IsLarge:bool)), Translation="0,-6,0", Rotation="0,0,6") {
    rotationMode=BMM_Additive
    rotationSpace=BCS_BoneSpace
    translationMode=BMM_Additive
    translationSpace=BCS_BoneSpace
    alphaBoolBlend {blendOption=Linear}
    boneToModify {boneName=pack}
    rotation {roll=6}
    translation {y=-6}
   }
   Output Pose(Result=Component To Local(ComponentPose=@n32)) {layerGroup=None}
   n33 = "Transform (Modify) Bone - Bone: spine_01"(
    ComponentPose=Local To Component(LocalPose=Local Space Ref Pose())
    Alpha=1
    Translation=Get spine_01_tran() -> (spine_01_tran:Vector)
    Rotation=Get spine_01_rot() -> (spine_01_rot:Rotator)
    Scale="1,1,1"
   ) {
    rotationMode=BMM_Replace
    rotationSpace=BCS_WorldSpace
    scaleSpace=BCS_BoneSpace
    translationMode=BMM_Replace
    translationSpace=BCS_WorldSpace
    alphaBoolBlend {blendOption=Linear}
    alphaScaleBiasClamp {inRange {max=0, min=-40}, outRange {max=0, min=1}}
    boneToModify {boneName=spine_01}
   }
   n34 = "Transform (Modify) Bone - Bone: spine_02"(
    ComponentPose=@n33
    Alpha=1
    Translation=Get spine_02_tran() -> (spine_02_tran:Vector)
    Rotation=Get spine_02_rot() -> (spine_02_rot:Rotator)
    Scale="1,1,1"
   ) {
    rotationMode=BMM_Replace
    rotationSpace=BCS_WorldSpace
    scaleSpace=BCS_BoneSpace
    translationMode=BMM_Replace
    translationSpace=BCS_WorldSpace
    alphaBoolBlend {blendOption=Linear}
    alphaScaleBiasClamp {inRange {max=0, min=-40}, outRange {max=0, min=1}}
    boneToModify {boneName=spine_02}
   }
   n35 = "Transform (Modify) Bone - Bone: spine_03"(
    ComponentPose=@n34
    Alpha=1
    Translation=Get spine_03_tran() -> (spine_03_tran:Vector)
    Rotation=Get spine_03_rot() -> (spine_03_rot:Rotator)
    Scale="1,1,1"
   ) {
    rotationMode=BMM_Replace
    rotationSpace=BCS_WorldSpace
    scaleSpace=BCS_BoneSpace
    translationMode=BMM_Replace
    translationSpace=BCS_WorldSpace
    alphaBoolBlend {blendOption=Linear}
    alphaScaleBiasClamp {inRange {max=0, min=-40}, outRange {max=0, min=1}}
    boneToModify {boneName=spine_03}
   }
   n36 = "Transform (Modify) Bone - Bone: spine_04"(
    ComponentPose=@n35
    Alpha=1
    Translation=Get spine_04_tran() -> (spine_04_tran:Vector)
    Rotation=Get spine_04_rot() -> (spine_04_rot:Rotator)
    Scale="1,1,1"
   ) {
    rotationMode=BMM_Replace
    rotationSpace=BCS_WorldSpace
    scaleSpace=BCS_BoneSpace
    translationMode=BMM_Replace
    translationSpace=BCS_WorldSpace
    alphaBoolBlend {blendOption=Linear}
    alphaScaleBiasClamp {inRange {max=0, min=-40}, outRange {max=0, min=1}}
    boneToModify {boneName=spine_04}
   }
   n37 = "Transform (Modify) Bone - Bone: spine_05"(
    ComponentPose=@n36
    Alpha=1
    Translation=Get spine_05_tran() -> (spine_05_tran:Vector)
    Rotation=Get spine_05_rot() -> (spine_05_rot:Rotator)
    Scale="1,1,1"
   ) {
    rotationMode=BMM_Replace
    rotationSpace=BCS_WorldSpace
    scaleSpace=BCS_BoneSpace
    translationMode=BMM_Replace
    translationSpace=BCS_WorldSpace
    alphaBoolBlend {blendOption=Linear}
    alphaScaleBiasClamp {inRange {max=0, min=-40}, outRange {max=0, min=1}}
    boneToModify {boneName=spine_05}
   }
   n38 = "Transform (Modify) Bone - Bone: neck_01"(
    ComponentPose=@n37
    Alpha=1
    Translation=Get neck_01_Tran() -> (neck_01_Tran:Vector)
    Rotation=Get neck_01_Rot() -> (neck_01_Rot:Rotator)
    Scale="1,1,1"
   ) {
    rotationMode=BMM_Replace
    rotationSpace=BCS_WorldSpace
    scaleSpace=BCS_BoneSpace
    translationMode=BMM_Replace
    translationSpace=BCS_WorldSpace
    alphaBoolBlend {blendOption=Linear}
    alphaScaleBiasClamp {inRange {max=0, min=-40}, outRange {max=0, min=1}}
    boneToModify {boneName=neck_01}
   }
   n39 = "Transform (Modify) Bone - Bone: neck_02"(
    ComponentPose=@n38
    Alpha=1
    Translation=Get neck_02_Tran() -> (neck_02_Tran:Vector)
    Rotation=Get neck_02_Rot() -> (neck_02_Rot:Rotator)
    Scale="1,1,1"
   ) {
    rotationMode=BMM_Replace
    rotationSpace=BCS_WorldSpace
    scaleSpace=BCS_BoneSpace
    translationMode=BMM_Replace
    translationSpace=BCS_WorldSpace
    alphaBoolBlend {blendOption=Linear}
    alphaScaleBiasClamp {inRange {max=0, min=-40}, outRange {max=0, min=1}}
    boneToModify {boneName=neck_02}
   }
   n40 = "Anim Dynamics - Chain: dyn_spine_05 -> dyn_pendant"!(ComponentPose=@n39, Alpha=1, GravityOverride:Vector="0,0,-980", ExternalForce="0,0,0") -> (Pose) {
    angularDampingOverride=0.9
    angularSpringConstant=300
    bChain=true
    bOverrideAngularDamping=true
    bOverrideLinearDamping=true
    bUseGravityOverride=true
    bUsePlanarLimit=false
    linearDampingOverride=0.9
    linearSpringConstant=300
    alphaBoolBlend {blendOption=Linear}
    boundBone {boneName=dyn_spine_05}
    chainEnd {boneName=dyn_pendant}
    componentAppliedLinearAccClamp=<3000 3000 3000>
    physicsBodyDefinitions [
     {
      boundBone {boneName=dyn_spine_05}
      boxExtents=<50 50 50>
      constraintSetup {angularLimitsMax=<0 5 5>, angularLimitsMin=<0 -5 -5>, angularTarget=<1 0 0>}
      localJointOffset=<0 0 0>
     }
     {
      boundBone {boneName=dyn_spine_04}
      boxExtents=<40 40 40>
      constraintSetup {angularLimitsMax=<0 5 5>, angularLimitsMin=<0 -5 -5>, angularTarget=<1 0 0>}
      localJointOffset=<0 0 0>
     }
     {
      boundBone {boneName=dyn_spine_03}
      boxExtents=<30 30 30>
      constraintSetup {angularLimitsMax=<0 5 5>, angularLimitsMin=<0 -5 -5>, angularTarget=<1 0 0>}
      localJointOffset=<0 0 0>
     }
     {
      boundBone {boneName=dyn_spine_02}
      boxExtents=<20 20 20>
      constraintSetup {angularLimitsMax=<0 5 5>, angularLimitsMin=<0 -5 -5>, angularTarget=<1 0 0>}
      localJointOffset=<0 0 0>
     }
     {
      boundBone {boneName=dyn_spine_01}
      boxExtents=<10 10 10>
      constraintSetup {angularLimitsMax=<0 10 10>, angularLimitsMin=<0 -10 -10>, angularTarget=<1 0 0>}
      localJointOffset=<0 0 0>
     }
     {
      boundBone {boneName=dyn_pendant}
      boxExtents=<10 10 10>
      constraintSetup {angularLimitsMax=<0 10 10>, angularLimitsMin=<0 -10 -10>, angularTarget=<1 0 0>}
      localJointOffset=<0 0 0>
     }
    ]
   }
   n41 = "Anim Dynamics - Chain: dyn_spine_05 -> dyn_neck_02"!(ComponentPose=@n40, Alpha=1, GravityOverride:Vector="0,0,-980", ExternalForce="0,0,0") -> (Pose) {
    angularDampingOverride=0.9
    angularSpringConstant=300
    bChain=true
    bOverrideAngularDamping=true
    bOverrideLinearDamping=true
    bUseGravityOverride=true
    bUsePlanarLimit=false
    linearDampingOverride=0.9
    linearSpringConstant=300
    alphaBoolBlend {blendOption=Linear}
    boundBone {boneName=dyn_spine_05}
    chainEnd {boneName=dyn_neck_02}
    componentAppliedLinearAccClamp=<3000 3000 3000>
    physicsBodyDefinitions [
     {
      boundBone {boneName=dyn_spine_05}
      boxExtents=<50 50 50>
      constraintSetup {angularLimitsMax=<0 5 5>, angularLimitsMin=<0 -5 -5>, angularTarget=<1 0 0>}
      localJointOffset=<0 0 0>
     }
     {
      boundBone {boneName=dyn_neck_01}
      boxExtents=<40 40 40>
      constraintSetup {angularLimitsMax=<0 5 5>, angularLimitsMin=<0 -5 -5>, angularTarget=<1 0 0>}
      localJointOffset=<0 0 0>
     }
     {
      boundBone {boneName=dyn_neck_02}
      boxExtents=<40 40 40>
      constraintSetup {angularLimitsMax=<0 5 5>, angularLimitsMin=<0 -5 -5>, angularTarget=<1 0 0>}
      localJointOffset=<0 0 0>
     }
    ]
   }
   n42 = "Constraint - dyn_spine_05 (PP)"!(ComponentPose=@n41, Alpha=0.8, ConstraintWeights_0=0.5, ConstraintWeights_1:float=0.5) -> (Pose) {
    alphaBoolBlend {blendOption=Linear}
    boneToModify {boneName=dyn_spine_05}
    constraintSetup [{targetBone {boneName=spine_04}}, {targetBone {boneName=spine_05}}]
    constraintWeights [1,1]
   }
   n43 = "Constraint - dyn_spine_04 (PP)"!(ComponentPose=@n42, Alpha=0.8, ConstraintWeights_0=0.5, ConstraintWeights_1:float=0.5) -> (Pose) {
    alphaBoolBlend {blendOption=Linear}
    boneToModify {boneName=dyn_spine_04}
    constraintSetup [{targetBone {boneName=spine_03}}, {targetBone {boneName=spine_04}}]
    constraintWeights [1,1]
   }
   n44 = "Constraint - dyn_spine_03 (PP)"!(ComponentPose=@n43, Alpha=0.8, ConstraintWeights_0=0.5, ConstraintWeights_1:float=0.5) -> (Pose) {
    alphaBoolBlend {blendOption=Linear}
    boneToModify {boneName=dyn_spine_03}
    constraintSetup [{targetBone {boneName=spine_02}}, {targetBone {boneName=spine_03}}]
    constraintWeights [1,1]
   }
   n45 = "Constraint - dyn_spine_02 (PP)"!(ComponentPose=@n44, Alpha=0.8, ConstraintWeights_0=1, ConstraintWeights_1:float=0) -> (Pose) {
    alphaBoolBlend {blendOption=Linear}
    boneToModify {boneName=dyn_spine_02}
    constraintSetup [{targetBone {boneName=spine_02}}, {targetBone {boneName=spine_03}}]
    constraintWeights [1,1]
   }
   n46 = "Constraint - dyn_spine_01 (PP)"!(ComponentPose=@n45, Alpha=0.5, ConstraintWeights_0=0.2, ConstraintWeights_1:float=0.8) -> (Pose) {
    alphaBoolBlend {blendOption=Linear}
    boneToModify {boneName=dyn_spine_01}
    constraintSetup [{targetBone {boneName=spine_01}}, {targetBone {boneName=spine_02}}]
    constraintWeights [1,1]
   }
   n47 = "Constraint - dyn_neck_01 (PP)"!(ComponentPose=@n46, Alpha=0.8, ConstraintWeights_0=0.2, ConstraintWeights_1:float=0.8) -> (Pose) {
    alphaBoolBlend {blendOption=Linear}
    boneToModify {boneName=dyn_neck_01}
    constraintSetup [{targetBone {boneName=neck_01}}, {targetBone {boneName=spine_05}}]
    constraintWeights [1,1]
   }
   n48 = "Constraint - dyn_neck_02 (PP)"!(ComponentPose=@n47, Alpha=0.8, ConstraintWeights_0=0.3, ConstraintWeights_1:float=0.7) -> (Pose) {
    alphaBoolBlend {blendOption=Linear}
    boneToModify {boneName=dyn_neck_02}
    constraintSetup [{targetBone {boneName=neck_02}}, {targetBone {boneName=spine_05}}]
    constraintWeights [1,1]
   }
   n49 = Anim Dynamics - Bone: dyn_spine_05_rib_r!(ComponentPose=@n48, Alpha=1, GravityOverride:Vector="0,0,0", ExternalForce="0,0,0") -> (Pose) {
    angularDampingOverride=0.8
    angularSpringConstant=3000
    bAngularSpring=true
    bOverrideAngularDamping=true
    bUseGravityOverride=true
    bUsePlanarLimit=false
    linearDampingOverride=0.4
    linearSpringConstant=300
    alphaBoolBlend {blendOption=Linear}
    boundBone {boneName=dyn_spine_05_rib_r}
    componentAppliedLinearAccClamp=<3000 3000 3000>
    componentLinearAccScale=<0.5 0.5 0.5>
    componentLinearVelScale=<0.5 0.5 0.5>
    physicsBodyDefinitions [{
      boundBone {boneName=dyn_spine_05_rib_r}
      boxExtents=<55 55 55>
      constraintSetup {angularLimitsMax=<0 10 0>, angularLimitsMin=<0 -10 0>, angularTarget=<1 0 0>}
      localJointOffset=<-1 0 0>
     }]
   }
   n50 = Anim Dynamics - Bone: dyn_spine_05_lib_l!(ComponentPose=@n49, Alpha=1, GravityOverride:Vector="0,0,0", ExternalForce="0,0,0") -> (Pose) {
    angularDampingOverride=0.8
    angularSpringConstant=3000
    bAngularSpring=true
    bOverrideAngularDamping=true
    bUseGravityOverride=true
    bUsePlanarLimit=false
    linearDampingOverride=0.4
    linearSpringConstant=300
    alphaBoolBlend {blendOption=Linear}
    boundBone {boneName=dyn_spine_05_lib_l}
    componentAppliedLinearAccClamp=<3000 3000 3000>
    componentLinearAccScale=<0.5 0.5 0.5>
    componentLinearVelScale=<0.5 0.5 0.5>
    physicsBodyDefinitions [{
      boundBone {boneName=dyn_spine_05_lib_l}
      boxExtents=<55 55 55>
      constraintSetup {angularLimitsMax=<0 10 0>, angularLimitsMin=<0 -10 0>, angularTarget=<1 0 0>}
      localJointOffset=<1 0 0>
     }]
   }
   n51 = Anim Dynamics - Bone: dyn_spine_04_rib_r!(ComponentPose=@n50, Alpha=1, GravityOverride:Vector="0,0,0", ExternalForce="0,0,0") -> (Pose) {
    angularDampingOverride=0.8
    angularSpringConstant=3000
    bAngularSpring=true
    bOverrideAngularDamping=true
    bUseGravityOverride=true
    bUsePlanarLimit=false
    linearDampingOverride=0.4
    linearSpringConstant=300
    alphaBoolBlend {blendOption=Linear}
    boundBone {boneName=dyn_spine_04_rib_r}
    componentAppliedLinearAccClamp=<3000 3000 3000>
    componentLinearAccScale=<0.5 0.5 0.5>
    componentLinearVelScale=<0.5 0.5 0.5>
    physicsBodyDefinitions [{
      boundBone {boneName=dyn_spine_04_rib_r}
      boxExtents=<55 55 55>
      constraintSetup {angularLimitsMax=<0 10 0>, angularLimitsMin=<0 -10 0>, angularTarget=<1 0 0>}
      localJointOffset=<-1 0 0>
     }]
   }
   n52 = Anim Dynamics - Bone: dyn_spine_04_lib_l!(ComponentPose=@n51, Alpha=1, GravityOverride:Vector="0,0,0", ExternalForce="0,0,0") -> (Pose) {
    angularDampingOverride=0.8
    angularSpringConstant=3000
    bAngularSpring=true
    bOverrideAngularDamping=true
    bUseGravityOverride=true
    bUsePlanarLimit=false
    linearDampingOverride=0.4
    linearSpringConstant=300
    alphaBoolBlend {blendOption=Linear}
    boundBone {boneName=dyn_spine_04_lib_l}
    componentAppliedLinearAccClamp=<3000 3000 3000>
    componentLinearAccScale=<0.5 0.5 0.5>
    componentLinearVelScale=<0.5 0.5 0.5>
    physicsBodyDefinitions [{
      boundBone {boneName=dyn_spine_04_lib_l}
      boxExtents=<55 55 55>
      constraintSetup {angularLimitsMax=<0 12 0>, angularLimitsMin=<0 -12 0>, angularTarget=<1 0 0>}
      localJointOffset=<1 0 0>
     }]
   }
   n53 = Anim Dynamics - Bone: dyn_spine_03_rib_r!(ComponentPose=@n52, Alpha=1, GravityOverride:Vector="0,0,0", ExternalForce="0,0,0") -> (Pose) {
    angularDampingOverride=0.8
    angularSpringConstant=3000
    bAngularSpring=true
    bOverrideAngularDamping=true
    bUseGravityOverride=true
    bUsePlanarLimit=false
    linearDampingOverride=0.4
    linearSpringConstant=300
    alphaBoolBlend {blendOption=Linear}
    boundBone {boneName=dyn_spine_03_rib_r}
    componentAppliedLinearAccClamp=<3000 3000 3000>
    componentLinearAccScale=<0.5 0.5 0.5>
    componentLinearVelScale=<0.5 0.5 0.5>
    physicsBodyDefinitions [{
      boundBone {boneName=dyn_spine_03_rib_r}
      boxExtents=<58 58 58>
      constraintSetup {angularLimitsMax=<0 10 0>, angularLimitsMin=<0 -10 0>, angularTarget=<1 0 0>}
      localJointOffset=<-1 0 0>
     }]
   }
   n54 = Anim Dynamics - Bone: dyn_spine_03_lib_l!(ComponentPose=@n53, Alpha=1, GravityOverride:Vector="0,0,0", ExternalForce="0,0,0") -> (Pose) {
    angularDampingOverride=0.8
    angularSpringConstant=3000
    bAngularSpring=true
    bOverrideAngularDamping=true
    bUseGravityOverride=true
    bUsePlanarLimit=false
    linearDampingOverride=0.4
    linearSpringConstant=300
    alphaBoolBlend {blendOption=Linear}
    boundBone {boneName=dyn_spine_03_lib_l}
    componentAppliedLinearAccClamp=<3000 3000 3000>
    componentLinearAccScale=<0.5 0.5 0.5>
    componentLinearVelScale=<0.5 0.5 0.5>
    physicsBodyDefinitions [{
      boundBone {boneName=dyn_spine_03_lib_l}
      boxExtents=<58 58 58>
      constraintSetup {angularLimitsMax=<0 10 0>, angularLimitsMin=<0 -10 0>, angularTarget=<1 0 0>}
      localJointOffset=<1 0 0>
     }]
   }
   n55 = Anim Dynamics - Bone: dyn_spine_02_rib_r!(ComponentPose=@n54, Alpha=1, GravityOverride:Vector="0,0,0", ExternalForce="0,0,0") -> (Pose) {
    angularDampingOverride=0.8
    angularSpringConstant=3000
    bAngularSpring=true
    bOverrideAngularDamping=true
    bUseGravityOverride=true
    bUsePlanarLimit=false
    linearDampingOverride=0.4
    linearSpringConstant=300
    alphaBoolBlend {blendOption=Linear}
    boundBone {boneName=dyn_spine_02_rib_r}
    componentAppliedLinearAccClamp=<3000 3000 3000>
    componentLinearAccScale=<0.5 0.5 0.5>
    componentLinearVelScale=<0.5 0.5 0.5>
    physicsBodyDefinitions [{
      boundBone {boneName=dyn_spine_02_rib_r}
      boxExtents=<70 70 70>
      constraintSetup {angularLimitsMax=<0 7 0>, angularLimitsMin=<0 -7 0>, angularTarget=<1 0 0>}
      localJointOffset=<-1 0 0>
     }]
   }
   n56 = Anim Dynamics - Bone: dyn_spine_02_lib_l!(ComponentPose=@n55, Alpha=1, GravityOverride:Vector="0,0,0", ExternalForce="0,0,0") -> (Pose) {
    angularDampingOverride=0.8
    angularSpringConstant=3000
    bAngularSpring=true
    bOverrideAngularDamping=true
    bUseGravityOverride=true
    bUsePlanarLimit=false
    linearDampingOverride=0.4
    linearSpringConstant=300
    alphaBoolBlend {blendOption=Linear}
    boundBone {boneName=dyn_spine_02_lib_l}
    componentAppliedLinearAccClamp=<3000 3000 3000>
    componentLinearAccScale=<0.5 0.5 0.5>
    componentLinearVelScale=<0.5 0.5 0.5>
    physicsBodyDefinitions [{
      boundBone {boneName=dyn_spine_02_lib_l}
      boxExtents=<70 70 70>
      constraintSetup {angularLimitsMax=<0 7 0>, angularLimitsMin=<0 -7 0>, angularTarget=<1 0 0>}
      localJointOffset=<1 0 0>
     }]
   }
   AnimGraphNode_StateMachine_0 = ShutterDown_Idle() {is=title AnimGraphNode_StateMachine}
   Save cached pose 'PostConstraint'(Pose=Apply Additive(Base=Component To Local(ComponentPose=@n56), Additive=@AnimGraphNode_StateMachine_0, Alpha=0.99) {alphaBoolBlend {blendOption=Linear}}) {cacheName=PostConstraint}
  ]}
 ShutterDown_Idle {
  entry=Idle
  path=AnimGraph.AnimGraphNode_StateMachine_0.ShutterDown_Idle
  ids=id
  nodes [
   AnimStateNode_0 = Idle()
   AnimStateNode_1 = ShutterDown() {is=title AnimStateNode, bAlwaysResetOnEntry=true}
   AnimStateTransitionNode_0 = AnimStateTransitionNode() {crossfadeDuration=0.1}
   AnimStateTransitionNode_1 = AnimStateTransitionNode() {crossfadeDuration=0.1}
  ]
  links [
   {
    dst=ShutterDown
    rule=AnimGraph.AnimGraphNode_StateMachine_0.ShutterDown_Idle.AnimStateTransitionNode_0.Transition
    src=Idle
    via=AnimStateTransitionNode_0
   }
   {
    dst=Idle
    rule=AnimGraph.AnimGraphNode_StateMachine_0.ShutterDown_Idle.AnimStateTransitionNode_1.Transition
    src=ShutterDown
    via=AnimStateTransitionNode_1
   }
  ]
 }
 Idle {path=AnimGraph.AnimGraphNode_StateMachine_0.ShutterDown_Idle.AnimStateNode_0.Idle, nodes [Output Animation Pose()]}
 ShutterDown {
  path=AnimGraph.AnimGraphNode_StateMachine_0.ShutterDown_Idle.AnimStateNode_1.ShutterDown
  nodes [Output Animation Pose(Result="Sequence Player (additive) 'ShutterDownSpine'"() {sequence=/BRCosmetics/Accessories/FORT_Backpacks/Backpack_M_MED_Quartz/Animation/ShutterDownSpine})]
 }
 Transition {
  path=AnimGraph.AnimGraphNode_StateMachine_0.ShutterDown_Idle.AnimStateTransitionNode_0.Transition
  nodes [Result(bCanEnterTransition=Get ShutterDownAnimation() -> (ShutterDownAnimation:bool))]
 }
 Transition {
  path=AnimGraph.AnimGraphNode_StateMachine_0.ShutterDown_Idle.AnimStateTransitionNode_1.Transition
  nodes [
   Result(bCanEnterTransition="float <= float"(A="Time Remaining (ratio) (ShutterDownSpine)"!() -> (ReturnValue:float) {is=title K2Node_AnimGetter}, B=0.1) -> (ReturnValue=~))
   "Time Remaining (ShutterDownSpine)"!() -> (ReturnValue:float) {is=title K2Node_AnimGetter}
  ]
 }
 BodySocketPosition {nodes [
   n1 = BodySocketPosition!() -> (then, SkeletalMesh:Scene Component Object Reference, InSocketName:Name) {is=title K2Node_FunctionEntry}
   n2 = BreakTransform(InTransform=GetSocketTransform(self=@n1, InSocketName=@n1.InSocketName, TransformSpace=RTS_World))
   > Return Node!(execute, Location:Vector=vector + vector(A=@n2, B=Get RootOffset() -> (RootOffset:Vector)), Rotation:Rotator=@n2.Rotation)
  ]}
 EventGraph {nodes [
   Event BlueprintUpdateAnimation()
   > Is Valid(InputObject=Get BodyMesh() -> (BodyMesh:Skeletal Mesh Component Object Reference))
   > n1 = Sequence!(execute) -> (then_0, then_1, then_2:Exec, then_3:Exec, then_4:Exec, then_5:Exec, then_6:Exec, then_7:Exec, then_8:Exec, then_9:Exec, then_10:Exec)
   n2 = Get BodyMesh()
   n3 = BodySocketPosition!(execute=@n1.then_1, self, SkeletalMesh:Scene Component Object Reference=@n2, InSocketName:Name=spine_02) -> (then, Location:Vector, Rotation:Rotator) {is=title K2Node_CallFunction}
   > Set spine_02_tran(spine_02_tran:Vector=@n3) -> (Output_Get:Vector)
   > Set spine_02_rot(spine_02_rot:Rotator=@n3.Rotation) -> (Output_Get:Rotator)
   n4 = GetSocketTransform(self=Get BodyMesh(), InSocketName=Backpack_BR, TransformSpace=RTS_World)
   n5 = GetSocketTransform(self=GetPartSkeletalMeshComponent(), InSocketName=root, TransformSpace=RTS_World)
   Set RootOffset(execute=@n1.then_10, RootOffset:Vector=vector - vector(A=BreakTransform(InTransform=@n5), B=BreakTransform(InTransform=@n4))) -> (Output_Get:Vector)
   n6 = BodySocketPosition(execute=@n1.then_2, SkeletalMesh=@n2, InSocketName=spine_03)
   > Set spine_03_tran(spine_03_tran:Vector=@n6) -> (Output_Get:Vector)
   > Set spine_03_rot(spine_03_rot:Rotator=@n6.Rotation) -> (Output_Get:Rotator)
   n7 = BodySocketPosition(execute=@n1.then_3, SkeletalMesh=@n2, InSocketName=spine_04)
   > Set spine_04_tran(spine_04_tran:Vector=@n7) -> (Output_Get:Vector)
   > Set spine_04_rot(spine_04_rot:Rotator=@n7.Rotation) -> (Output_Get:Rotator)
   n8 = BodySocketPosition(execute=@n1.then_4, SkeletalMesh=@n2, InSocketName=spine_05)
   > Set spine_05_tran(spine_05_tran:Vector=@n8) -> (Output_Get:Vector)
   > Set spine_05_rot(spine_05_rot:Rotator=@n8.Rotation) -> (Output_Get:Rotator)
   n9 = Get bIsParachuteOpened()
   Set SkyDivingWihtoutParachute(execute=@n1.then_8, SkyDivingWihtoutParachute:bool=AND Boolean(A=Get bIsSkyDiving(), B=NOT Boolean(A=@n9))) -> (Output_Get:bool)
   Set dyn_Spine_02_alpha_correction(
    execute=@n1.then_9
    dyn_Spine_02_alpha_correction:double=GetAlphaFromReferenceTransformDelta(
     SkeletalMeshComponent=Get BodyMesh()
     BoneName=spine_05
     BaseName=spine_01
     DeltaComponentType=QuaternionTwist
     InputRange=(X=-12,Y=-50)
     TwistAxis="0,0,1"
    )
   ) -> (Output_Get:double)
   n10 = BodySocketPosition(execute=@n1.then_5, SkeletalMesh=@n2, InSocketName=neck_01)
   > Set neck_01_Tran(neck_01_Tran:Vector=@n10) -> (Output_Get:Vector)
   > Set neck_01_Rot(neck_01_Rot:Rotator=@n10.Rotation) -> (Output_Get:Rotator)
   n11 = BodySocketPosition(execute=@n1.then_0, SkeletalMesh=@n2, InSocketName=spine_01)
   > Set spine_01_tran(spine_01_tran:Vector=@n11) -> (Output_Get:Vector)
   > Set spine_01_rot(spine_01_rot:Rotator=@n11.Rotation) -> (Output_Get:Rotator)
   n12 = Vector Length(A=Get PawnVelocity())
   n13 = GetGameTimeInSeconds()
   n14 = float + float(A="Sin (Radians)"(A=float * float(A=@n13, B=1.2) {is=title K2Node_PromotableOperator}), B=1)
   Set Breathe_alpha(
    execute=@n1.then_7
    Breathe_alpha:double=float * float(A=float / float(A=@n14, B=2), B=MapRangeClamped(Value=@n12, InRangeA=5, InRangeB=60, OutRangeA=1, OutRangeB=0))
   ) -> (Output_Get:double)
   n15 = BodySocketPosition(execute=@n1.then_6, SkeletalMesh=@n2, InSocketName=neck_02)
   > Set neck_02_Tran(neck_02_Tran:Vector=@n15) -> (Output_Get:Vector)
   > Set neck_02_Rot(neck_02_Rot:Rotator=@n15.Rotation) -> (Output_Get:Rotator)
   n16 = Event BlueprintProcessAnimInputEvent()
   n17 = IsNamed(self=@n16.AnimInputEvent, InEventName=aie.PlayerVitality.ShieldsIncreased)
   n18 = Get As Player Pawn Athena() -> (As Player Pawn Athena:Player Pawn Athena Object Reference)
   n19 = Contains(
    SearchIn=GetDisplayName(Object=Get CurrentWeapon!(self:Fort Pawn Object Reference=@n18) -> (CurrentWeapon:Fort Weapon Object Reference))
    Substring=SlurpJuice
   )
   > Branch(Condition=OR Boolean!(
     A=IsNamed(self=@n16.AnimInputEvent, InEventName=aie.ChestOpened)
     B=IsNamed(self=@n16.AnimInputEvent, InEventName=aie.EnemyPlayerEliminated)
     C:bool=AND Boolean(A=@n17, B=@n19)
    ) -> (ReturnValue))
   > Set ShutterDownAnimation(ShutterDownAnimation:bool=true) -> (Output_Get:bool)
   Event BlueprintPostEvaluateAnimation() {is=title K2Node_Event}
   > Set ShutterDownAnimation(ShutterDownAnimation=false)
   Event OnCharacterCustomizationLoaded()
   > n20 = Set Owning Pawn(Owning Pawn:Pawn Object Reference=TryGetPawnOwner()) -> (Output_Get:Pawn Object Reference)
   > Is Valid(InputObject=@n20)
   > n21 = Sequence!(execute) -> (then_0, then_1, then_2)
   n22 = FindBodyType!(execute=@n21.then_1, Object:Object Reference=Get Owning Pawn() -> (Owning Pawn:Pawn Object Reference)) -> (True:Exec, False:Exec, OutBodyType:EFortCustomBodyType Enum=NONE)
   > Set IsLarge(IsLarge:bool="Equal (Enum)"(A:EFortCustomBodyType Enum=@n22, B:EFortCustomBodyType Enum=Large)) -> (Output_Get:bool)
   n23 = Cast To FortPlayerPawn(execute=@n21.then_0, Object=Get Owning Pawn())
   > n24 = Set Owning Player Pawn(Owning Player Pawn:Fort Player Pawn Object Reference=@n23) -> (Output_Get:Fort Player Pawn Object Reference)
   > Set BodyMesh(BodyMesh=GetSkeletalMeshForPartType(self=@n24))
   n25 = FindGender!(execute=@n21.then_2, Object:Object Reference=Get Owning Pawn()) -> (True:Exec, False:Exec, OutGender:EFortCustomGender Enum=Invalid)
   > Set isFemale(isFemale:bool="Equal (Enum)"(A:EFortCustomGender Enum=@n25, B:EFortCustomGender Enum=Female)) -> (Output_Get:bool)
   TryGetPawnOwner()
  ]}
]
variables [
 As Player Pawn Athena {type=PlayerPawn_Athena_C}
 BodyMesh {type=SkeletalMeshComponent}
 Breathe_alpha {type=number}
 Dyn_Spine_01_Rotation {type=Rotator, default=<rot 5 10 0>}
 dyn_Spine_02_alpha_correction {type=number}
 isFemale {type=boolean}
 IsLarge {type=boolean}
 neck_01_Rot {type=Rotator}
 neck_01_Tran {type=Vector}
 neck_02_Rot {type=Rotator}
 neck_02_Tran {type=Vector}
 Owning Pawn {type=Pawn, default=null}
 Owning Player Pawn {type=FortPlayerPawn}
 RootOffset {type=Vector}
 ShutterDownAnimation {type=boolean}
 SkyDivingWihtoutParachute {type=boolean}
 spine_01_rot {type=Rotator}
 spine_01_tran {type=Vector}
 spine_02_rot {type=Rotator}
 spine_02_tran {type=Vector}
 spine_03_rot {type=Rotator}
 spine_03_tran {type=Vector}
 spine_04_rot {type=Rotator}
 spine_04_tran {type=Vector}
 spine_05_rot {type=Rotator}
 spine_05_tran {type=Vector}
]
