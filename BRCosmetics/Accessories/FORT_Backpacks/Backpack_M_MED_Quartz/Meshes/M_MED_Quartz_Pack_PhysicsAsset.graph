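# Physics asset for the male medium Quartz backpack cosmetic mesh in the Battle Royale cosmetics set.
# It holds a single body, bound to the pack's LOD0 mesh bone, with one rotated capsule approximating the whole backpack.
# It has no constraints, so the pack acts as one rigid shape with no jointed or secondary-motion parts.
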PhysicsAsset /BRCosmetics/Accessories/FORT_Backpacks/Backpack_M_MED_Quartz/Meshes/M_MED_Quartz_Pack_PhysicsAsset
bodies {Backpack_Quartz_Male_MESH_LOD0 [{shapeType=Capsule, length=17.7135, radius=14.0733, center=<0 -2.6389 -0.1964>, rotation=<rot -1.0582 -90 0.0002>}]}
constraints []
